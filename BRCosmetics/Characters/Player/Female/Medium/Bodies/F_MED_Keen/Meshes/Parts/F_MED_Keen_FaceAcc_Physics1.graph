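# Physics asset for the dynamic hair on the Keen female medium body's face accessory mesh.
# Capsule bodies form a central hair chain that branches into four tips after the fourth bone, plus five-link left and right side chains, each rooted to the head through a tightly limited base.
# Central links swing mostly on one axis with near-zero second swing, while side strands have narrower limits.
# Unconstrained head, neck, clavicle, upper-arm and upper-spine capsules provide body collision shapes.
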PhysicsAsset /BRCosmetics/Characters/Player/Female/Medium/Bodies/F_MED_Keen/Meshes/Parts/F_MED_Keen_FaceAcc_Physics1
bodies {
 clavicle_l [{shapeType=Capsule, length=11, radius=4, shapeName=clavicle_l_capsule, center=<7.1066 0.0001 0>, rotation=<rot -90 0.0297 -0.0299>}]
 clavicle_r [{shapeType=Capsule, length=11, radius=4, shapeName=clavicle_r_capsule, center=<-6.8783 0.0003 0>, rotation=<rot -90 -0.0197 0.0195>}]
 dyn_Hair_01 [{shapeType=Capsule, length=4, radius=3, shapeName=dyn_Hair_01_capsule, center=<0.123 -1.3037 -0.0073>, rotation=<rot 86.5788 0.0001 -95.399>}]
 dyn_Hair_02 [{shapeType=Capsule, length=6, radius=3.25, shapeName=dyn_Hair_02_capsule, center=<0.3815 -4.0441 -0.0228>, rotation=<rot 86.5765 0 -95.3986>}]
 dyn_Hair_03 [{
   shapeType=Capsule
   length=4.25
   radius=4
   shapeName=dyn_Hair_03_capsule
   center=<0.0309 -3.5069 -0.0019>
   rotation=<rot 86.578 0.0006 -82.8332>
  }]
 dyn_Hair_04 [{
   shapeType=Capsule
   length=8.342
   radius=5.158
   shapeName=dyn_Hair_04_capsule
   center=<-0.1194 -8.0818 0.0072>
   rotation=<rot 86.5778 0.0002 -91.5046>
  }]
 dyn_Hair_05 [{
   shapeType=Capsule
   length=5.2759
   radius=5.2355
   shapeName=dyn_Hair_05_capsule
   center=<0.2656 -2.8157 -0.0159>
   rotation=<rot 86.5781 0.0005 -95.3987>
  }]
 dyn_Hair_06 [{
   shapeType=Capsule
   length=7.813
   radius=2.0235
   shapeName=dyn_Hair_06_capsule
   center=<-0.3105 -2.8111 0.0185>
   rotation=<rot 86.5789 0.0007 -83.6844>
  }]
 dyn_Hair_07 [{
   shapeType=Capsule
   length=6.0865
   radius=2.5
   shapeName=dyn_Hair_07_capsule
   center=<-0.7615 -2.7234 0.0455>
   rotation=<rot 86.5788 0.0006 -74.3501>
  }]
 dyn_Hair_08 [{
   shapeType=Capsule
   length=6.2293
   radius=1.9925
   shapeName=dyn_Hair_08_capsule
   center=<0.2417 -2.5626 -0.0144>
   rotation=<rot 86.5818 0.0005 -95.397>
  }]
 dyn_Hair_base [{shapeType=Capsule, length=0.505, radius=0.505, shapeName=dyn_Hair_base_capsule, rotation=<rot 0 -90 0>}]
 dyn_Hair_l_01 [{
   shapeType=Capsule
   length=3.8165
   radius=1.4335
   shapeName=dyn_Hair_l_01_capsule
   center=<-2.1504 0.1625 0.1269>
   rotation=<rot 86.6222 0.0004 4.3141>
  }]
 dyn_Hair_l_02 [{
   shapeType=Capsule
   length=3.8165
   radius=1.4335
   shapeName=dyn_Hair_l_02_capsule
   center=<-2.1503 0.1621 0.1286>
   rotation=<rot 86.5786 -0.0001 4.3024>
  }]
 dyn_Hair_l_03 [{
   shapeType=Capsule
   length=3.8165
   radius=1.4335
   shapeName=dyn_Hair_l_03_capsule
   center=<-2.1503 0.1621 0.1286>
   rotation=<rot 86.5786 -0.0001 4.3024>
  }]
 dyn_Hair_l_04 [{
   shapeType=Capsule
   length=3.8165
   radius=1.4335
   shapeName=dyn_Hair_l_04_capsule
   center=<-2.1503 0.1621 0.1286>
   rotation=<rot 86.5786 -0.0001 4.3024>
  }]
 dyn_Hair_l_05 [{
   shapeType=Capsule
   length=3.8165
   radius=1.4335
   shapeName=dyn_Hair_l_05_capsule
   center=<-2.0547 1.4346 0.1229>
   rotation=<rot 86.5779 0 27.5249>
  }]
 dyn_Hair_l_base [{shapeType=Capsule, length=0.505, radius=0.505, shapeName=dyn_Hair_l_base_capsule, rotation=<rot 0 -90 0>}]
 dyn_Hair_r_01 [{
   shapeType=Capsule
   length=3.7781
   radius=1.4719
   shapeName=dyn_Hair_r_01_capsule
   center=<1.4574 -0.1101 -0.086>
   rotation=<rot 86.6205 -0.0002 4.313>
  }]
 dyn_Hair_r_02 [{
   shapeType=Capsule
   length=4.0598
   radius=1.5906
   shapeName=dyn_Hair_r_02_capsule
   center=<3.2068 -0.2427 -0.1916>
   rotation=<rot 86.5791 -0.0001 -3.1268>
  }]
 dyn_Hair_r_03 [{
   shapeType=Capsule
   length=4.7052
   radius=1.5414
   shapeName=dyn_Hair_r_03_capsule
   center=<1.2568 -0.0952 -0.0887>
   rotation=<rot 86.338 171.6627 175.9674>
  }]
 dyn_Hair_r_04 [{
   shapeType=Capsule
   length=3.5468
   radius=1.5414
   shapeName=dyn_Hair_r_04_capsule
   center=<1.2444 -0.0394 -0.0743>
   rotation=<rot 86.5795 -0.0001 4.3203>
  }]
 dyn_Hair_r_05 [{
   shapeType=Capsule
   length=3.8546
   radius=1.2336
   shapeName=dyn_Hair_r_05_capsule
   center=<1.6367 -1.192 -0.0976>
   rotation=<rot 86.5768 0.0001 33.2323>
  }]
 dyn_Hair_r_base [{shapeType=Capsule, length=0.505, radius=0.505, shapeName=dyn_Hair_r_base_capsule, rotation=<rot 0 -90 0>}]
 head [{shapeType=Capsule, length=7.2457, radius=7.7917, center=<5.8418 0.3192 -0.2521>, rotation=<rot 2.7219 -86.018 -90.8388>}]
 neck_01 [{shapeType=Capsule, length=3.75, radius=6.5, shapeName=neck_01_capsule, center=<3.7232 -0.8457 0.0227>, rotation=<rot -89.65 0.0078 -14.2972>}]
 spine_05 [
  {shapeType=Capsule, length=17.433, radius=8.7365, shapeName=spine_05_capsule, center=<0.9101 -1.975 0>}
  {shapeType=Capsule, length=38.5886, radius=6.3358, shapeName=spine_05_capsule, center=<3.9534 -8.1503 3.4697>, rotation=<rot -56.6163 0 0>}
  {
   shapeType=Capsule
   length=39.1557
   radius=5.7687
   shapeName=spine_05_capsule
   center=<3.5069 -8.1502 -5.3219>
   rotation=<rot -55.8692 -179.9999 179.9999>
  }
 ]
 upperarm_l [{shapeType=Capsule, length=28.3221, radius=3.6779, shapeName=upperarm_l_capsule, center=<15.1711 -0.0004 0.6774>, rotation=<rot -87.4435 0 0>}]
 upperarm_r [{
   shapeType=Capsule
   length=28.3221
   radius=3.6779
   shapeName=upperarm_r_capsule
   center=<-13.7046 -0.0002 -0.6119>
   rotation=<rot -87.4435 0 0>
  }]
}
constraints [
 {bone1Name=dyn_Hair_base, bone2Name=head, swing1LimitDegrees=1, swing2LimitDegrees=0.1, twistLimitDegrees=10}
 {bone1Name=dyn_Hair_01, bone2Name=dyn_Hair_base, swing1LimitDegrees=5, swing2LimitDegrees=0.01, twistLimitDegrees=5}
 {bone1Name=dyn_Hair_02, bone2Name=dyn_Hair_01, swing1LimitDegrees=30, swing2LimitDegrees=0.01, twistLimitDegrees=15}
 {bone1Name=dyn_Hair_03, bone2Name=dyn_Hair_02, swing1LimitDegrees=30, swing2LimitDegrees=0.01, twistLimitDegrees=15}
 {bone1Name=dyn_Hair_04, bone2Name=dyn_Hair_03, swing1LimitDegrees=30, swing2LimitDegrees=0.01, twistLimitDegrees=15}
 {bone1Name=dyn_Hair_05, bone2Name=dyn_Hair_04, swing1LimitDegrees=30, swing2LimitDegrees=0.01, twistLimitDegrees=15}
 {bone1Name=dyn_Hair_06, bone2Name=dyn_Hair_04, swing1LimitDegrees=20, swing2LimitDegrees=0.01, twistLimitDegrees=10}
 {bone1Name=dyn_Hair_07, bone2Name=dyn_Hair_04, swing1LimitDegrees=20, swing2LimitDegrees=0.01, twistLimitDegrees=10}
 {bone1Name=dyn_Hair_08, bone2Name=dyn_Hair_04, swing1LimitDegrees=20, swing2LimitDegrees=0.01, twistLimitDegrees=10}
 {bone1Name=dyn_Hair_l_base, bone2Name=head, swing1LimitDegrees=1, swing2LimitDegrees=1, twistLimitDegrees=1}
 {bone1Name=dyn_Hair_l_01, bone2Name=dyn_Hair_l_base, swing1LimitDegrees=15, swing2LimitDegrees=5, twistLimitDegrees=5}
 {bone1Name=dyn_Hair_l_02, bone2Name=dyn_Hair_l_01, swing1LimitDegrees=15, swing2LimitDegrees=5, twistLimitDegrees=5}
 {bone1Name=dyn_Hair_l_03, bone2Name=dyn_Hair_l_02, swing1LimitDegrees=15, swing2LimitDegrees=5, twistLimitDegrees=5}
 {bone1Name=dyn_Hair_l_04, bone2Name=dyn_Hair_l_03, swing1LimitDegrees=15, swing2LimitDegrees=5, twistLimitDegrees=5}
 {bone1Name=dyn_Hair_l_05, bone2Name=dyn_Hair_l_04, swing1LimitDegrees=15, swing2LimitDegrees=5, twistLimitDegrees=5}
 {bone1Name=dyn_Hair_r_base, bone2Name=head, swing1LimitDegrees=1, swing2LimitDegrees=1, twistLimitDegrees=1}
 {bone1Name=dyn_Hair_r_01, bone2Name=dyn_Hair_r_base, swing1LimitDegrees=15, swing2LimitDegrees=5, twistLimitDegrees=5}
 {bone1Name=dyn_Hair_r_02, bone2Name=dyn_Hair_r_01, swing1LimitDegrees=15, swing2LimitDegrees=5, twistLimitDegrees=5}
 {bone1Name=dyn_Hair_r_03, bone2Name=dyn_Hair_r_02, swing1LimitDegrees=15, swing2LimitDegrees=5, twistLimitDegrees=5}
 {bone1Name=dyn_Hair_r_04, bone2Name=dyn_Hair_r_03, swing1LimitDegrees=15, swing2LimitDegrees=5, twistLimitDegrees=5}
 {bone1Name=dyn_Hair_r_05, bone2Name=dyn_Hair_r_04, swing1LimitDegrees=15, swing2LimitDegrees=5, twistLimitDegrees=5}
]
