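# Post-process style AnimBlueprint that adds secondary hair motion to the Keen female character's FaceAcc mesh.
# The parent's copied pose gets side-strand AnimDynamics and right-shoulder corrections, then is cached for two AnimDynamics-plus-RigidBody branches, selected by whether non-long backbling is worn.
# Frontend state alphas switch physics assets within each branch.
# The event graph updates those alphas, skydive gravity and a no-rigid-body fallback offset used in the backpack branch.
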AnimBlueprint /BRCosmetics/Characters/Player/Female/Medium/Bodies/F_MED_Keen/Meshes/Parts/F_MED_Keen_FaceAcc_AnimBP
target_skeleton=&/F_MED_Keen_FaceAcc_Skeleton
graphs [
 AnimGraph {nodes [
   "Constraint - dyn_Hair_02 (RR)"!(ComponentPose, Alpha=1, ConstraintWeights_0=0.6, ConstraintWeights_1:float=0.4) -> (Pose) {
    alphaBoolBlend {blendOption=Linear}
    boneToModify {boneName=dyn_Hair_02}
    constraintSetup [{transformType=Rotation, targetBone {boneName=spine_05}}, {transformType=Rotation, perAxis {bZ=false}, targetBone {boneName=dyn_Hair_03}}]
    constraintWeights [1,1]
   }
   n1 = Get Is In Frontend() -> (Is In Frontend:double)
   "Constraint - dyn_Hair_04 (PP)"!(ComponentPose, Alpha=@n1, ConstraintWeights_0=0.6, ConstraintWeights_1:float=0.4) -> (Pose) {
    alphaBoolBlend {blendOption=Linear}
    boneToModify {boneName=dyn_Hair_04}
    constraintSetup [{targetBone {boneName=spine_05}}, {targetBone {boneName=dyn_Hair_03}}]
    constraintWeights [1,1]
   }
   n2 = Get Is In Frontend()
   "Constraint - dyn_Hair_04 (RR)"!(ComponentPose, Alpha=@n2, ConstraintWeights_0=0.6, ConstraintWeights_1:float=0.4) -> (Pose) {
    alphaBoolBlend {blendOption=Linear}
    boneToModify {boneName=dyn_Hair_04}
    constraintSetup [{transformType=Rotation, targetBone {boneName=spine_05}}, {transformType=Rotation, perAxis {bZ=false}, targetBone {boneName=dyn_Hair_03}}]
    constraintWeights [1,1]
   }
   n3 = Anim Dynamics - Bone: dyn_Hair_01(ComponentPose=Local To Component(LocalPose=Use cached pose 'BackpackCache'()), Alpha=0.4, ExternalForce="0,0,0") {
    angularDampingOverride=0.999
    bAngularSpring=true
    bGravityOverrideInSimSpace=true
    bOverrideAngularDamping=true
    bUseGravityOverride=true
    simulationSpace=RootRelative
    alphaBoolBlend {blendOption=Linear}
    boundBone {boneName=dyn_Hair_01}
    componentAppliedLinearAccClamp=<2500 2500 2500>
    componentLinearAccScale=<0.3 0.3 0.35>
    componentLinearVelScale=<0.35 0.35 0.35>
    gravityOverride {z=-600}
    physicsBodyDefinitions [{
      sphereCollisionRadius=0
      boundBone {boneName=dyn_Hair_01}
      boxExtents=<5 10 10>
      constraintSetup {angularTargetAxis=AxisY, angularLimitsMax=<5 1 20>, angularLimitsMin=<-5 -1 -7>, angularTarget=<0 0 0>}
      localJointOffset=<-5 -0.1 -0.5>
     }]
    relativeSpaceBone {boneName=dyn_Hair_base}
   }
   n4 = Anim Dynamics - Bone: dyn_Hair_05(ComponentPose=@n3, Alpha=0.4, ExternalForce="0,0,0") {
    angularDampingOverride=0.999
    bAngularSpring=true
    bGravityOverrideInSimSpace=true
    bOverrideAngularDamping=true
    bUseGravityOverride=true
    simulationSpace=RootRelative
    alphaBoolBlend {blendOption=Linear}
    boundBone {boneName=dyn_Hair_05}
    componentAppliedLinearAccClamp=<800 800 800>
    componentLinearAccScale=<0.3 0.3 0.35>
    componentLinearVelScale=<0.35 0.35 0.35>
    gravityOverride {z=-800}
    physicsBodyDefinitions [{
      sphereCollisionRadius=0
      boundBone {boneName=dyn_Hair_05}
      boxExtents=<10 5 10>
      constraintSetup {angularTargetAxis=AxisY, angularLimitsMax=<30 20 60>, angularLimitsMin=<-30 -20 -7>, angularTarget=<0 0 0>}
      localJointOffset=<-0.1 -5 1>
     }]
    relativeSpaceBone {boneName=dyn_Hair_03}
   }
   n5 = Anim Dynamics - Bone: dyn_Hair_02(ComponentPose=@n4, Alpha=0.4, ExternalForce="0,0,0") {
    angularDampingOverride=0.999
    bAngularSpring=true
    bGravityOverrideInSimSpace=true
    bOverrideAngularDamping=true
    bUseGravityOverride=true
    simulationSpace=RootRelative
    alphaBoolBlend {blendOption=Linear}
    boundBone {boneName=dyn_Hair_02}
    componentAppliedLinearAccClamp=<2500 2500 2500>
    componentLinearAccScale=<0.3 0.3 0.35>
    componentLinearVelScale=<0.35 0.35 0.35>
    gravityOverride {z=-600}
    physicsBodyDefinitions [{
      sphereCollisionRadius=0
      boundBone {boneName=dyn_Hair_02}
      boxExtents=<10 5 10>
      constraintSetup {angularTargetAxis=AxisY, angularLimitsMax=<5 1 45>, angularLimitsMin=<-5 -1 -7>, angularTarget=<0 0 0>}
      localJointOffset=<-2 -4 -0.5>
     }]
    relativeSpaceBone {boneName=dyn_Hair_01}
   }
   n6 = Anim Dynamics - Bone: dyn_Hair_06(ComponentPose=@n5, Alpha=0.4, ExternalForce="0,0,0") {
    angularDampingOverride=0.999
    bAngularSpring=true
    bGravityOverrideInSimSpace=true
    bOverrideAngularDamping=true
    bUseGravityOverride=true
    simulationSpace=RootRelative
    alphaBoolBlend {blendOption=Linear}
    boundBone {boneName=dyn_Hair_06}
    componentAppliedLinearAccClamp=<800 800 800>
    componentLinearAccScale=<0.3 0.3 0.35>
    componentLinearVelScale=<0.35 0.35 0.35>
    gravityOverride {z=-800}
    physicsBodyDefinitions [{
      sphereCollisionRadius=0
      boundBone {boneName=dyn_Hair_06}
      boxExtents=<10 5 10>
      constraintSetup {angularTargetAxis=AxisY, angularLimitsMax=<30 20 60>, angularLimitsMin=<-30 -20 -7>, angularTarget=<0 0 0>}
      localJointOffset=<-0.1 -5 1>
     }]
    relativeSpaceBone {boneName=dyn_Hair_03}
   }
   n7 = Anim Dynamics - Bone: dyn_Hair_03(ComponentPose=@n6, Alpha=0.4, ExternalForce="0,0,0") {
    angularDampingOverride=0.999
    bAngularSpring=true
    bGravityOverrideInSimSpace=true
    bOverrideAngularDamping=true
    bUseGravityOverride=true
    simulationSpace=RootRelative
    alphaBoolBlend {blendOption=Linear}
    boundBone {boneName=dyn_Hair_03}
    componentAppliedLinearAccClamp=<800 800 800>
    componentLinearAccScale=<0.3 0.3 0.35>
    componentLinearVelScale=<0.35 0.35 0.35>
    gravityOverride {z=-800}
    physicsBodyDefinitions [{
      sphereCollisionRadius=0
      boundBone {boneName=dyn_Hair_03}
      boxExtents=<10 5 10>
      constraintSetup {angularTargetAxis=AxisY, angularLimitsMax=<30 20 60>, angularLimitsMin=<-30 -20 -7>, angularTarget=<0 0 0>}
      localJointOffset=<-0.1 -5 1>
     }]
    relativeSpaceBone {boneName=dyn_Hair_02}
   }
   n8 = Anim Dynamics - Bone: dyn_Hair_07(ComponentPose=@n7, Alpha=0.4, ExternalForce="0,0,0") {
    angularDampingOverride=0.999
    bAngularSpring=true
    bGravityOverrideInSimSpace=true
    bOverrideAngularDamping=true
    bUseGravityOverride=true
    simulationSpace=RootRelative
    alphaBoolBlend {blendOption=Linear}
    boundBone {boneName=dyn_Hair_07}
    componentAppliedLinearAccClamp=<800 800 800>
    componentLinearAccScale=<0.3 0.3 0.35>
    componentLinearVelScale=<0.35 0.35 0.35>
    gravityOverride {z=-800}
    physicsBodyDefinitions [{
      sphereCollisionRadius=0
      boundBone {boneName=dyn_Hair_07}
      boxExtents=<10 5 10>
      constraintSetup {angularTargetAxis=AxisY, angularLimitsMax=<30 20 60>, angularLimitsMin=<-30 -20 -7>, angularTarget=<0 0 0>}
      localJointOffset=<-0.1 -5 1>
     }]
    relativeSpaceBone {boneName=dyn_Hair_03}
   }
   n9 = Anim Dynamics - Bone: dyn_Hair_04(ComponentPose=@n8, Alpha=0.4, ExternalForce="0,0,0") {
    angularDampingOverride=0.999
    bAngularSpring=true
    bGravityOverrideInSimSpace=true
    bOverrideAngularDamping=true
    bUseGravityOverride=true
    simulationSpace=RootRelative
    alphaBoolBlend {blendOption=Linear}
    boundBone {boneName=dyn_Hair_04}
    componentAppliedLinearAccClamp=<800 800 800>
    componentLinearAccScale=<0.3 0.3 0.35>
    componentLinearVelScale=<0.35 0.35 0.35>
    gravityOverride {z=-800}
    physicsBodyDefinitions [{
      sphereCollisionRadius=0
      boundBone {boneName=dyn_Hair_04}
      boxExtents=<10 5 10>
      constraintSetup {angularTargetAxis=AxisY, angularLimitsMax=<30 20 60>, angularLimitsMin=<-30 -20 -7>, angularTarget=<0 0 0>}
      localJointOffset=<-0.1 -5 1>
     }]
    relativeSpaceBone {boneName=dyn_Hair_03}
   }
   n10 = Anim Dynamics - Bone: dyn_Hair_08(ComponentPose=@n9, Alpha=0.4, ExternalForce="0,0,0") {
    angularDampingOverride=0.999
    bAngularSpring=true
    bGravityOverrideInSimSpace=true
    bOverrideAngularDamping=true
    bUseGravityOverride=true
    simulationSpace=RootRelative
    alphaBoolBlend {blendOption=Linear}
    boundBone {boneName=dyn_Hair_08}
    componentAppliedLinearAccClamp=<800 800 800>
    componentLinearAccScale=<0.3 0.3 0.35>
    componentLinearVelScale=<0.35 0.35 0.35>
    gravityOverride {z=-800}
    physicsBodyDefinitions [{
      sphereCollisionRadius=0
      boundBone {boneName=dyn_Hair_08}
      boxExtents=<10 5 10>
      constraintSetup {angularTargetAxis=AxisY, angularLimitsMax=<30 20 60>, angularLimitsMin=<-30 -20 -7>, angularTarget=<0 0 0>}
      localJointOffset=<-0.1 -5 1>
     }]
    relativeSpaceBone {boneName=dyn_Hair_03}
   }
   n11 = Get gravityOverride() -> (gravityOverride:Vector)
   n12 = RigidBody!(ComponentPose=@n10, Alpha=Get Is NOT In Frontend() -> (Is NOT In Frontend:double), OverrideWorldGravity:Vector=@n11, ExternalForce) -> (Pose) {
    bClampLinearTranslationLimitToRefPose=true
    bOverrideWorldGravity=true
    overridePhysicsAsset=&/F_MED_Keen_FaceAcc_Physics
    simulationSpace=BaseBoneSpace
    alphaBoolBlend {blendOption=Linear}
    baseBoneRef {boneName=spine_02}
    componentAppliedLinearAccClamp=<800 800 800>
    componentLinearAccScale=<0.4 0.4 0.45>
    componentLinearVelScale=<0.45 0.45 0.45>
    overrideWorldGravity {z=-980}
    simSpaceSettings {
     maxAngularAcceleration=1000
     maxAngularVelocity=0.3
     maxLinearAcceleration=700
     maxLinearVelocity=500
     velocityScaleZ=0
     worldAlpha=0.1
     externalLinearDragV=<0.5 0.5>
    }
   }
   n13 = RigidBody(ComponentPose=@n12, Alpha=@n2, OverrideWorldGravity=@n11) {
    bClampLinearTranslationLimitToRefPose=true
    bOverrideWorldGravity=true
    overridePhysicsAsset=&/F_MED_Keen_FaceAcc_Physics_Frontend
    simulationSpace=BaseBoneSpace
    alphaBoolBlend {blendOption=Linear}
    baseBoneRef {boneName=pelvis}
    componentAppliedLinearAccClamp=<5000 5000 5000>
    componentLinearAccScale=<2 2 2>
    componentLinearVelScale=<2 2 2>
    overrideWorldGravity {z=-980}
    simSpaceSettings {
     maxAngularAcceleration=1000
     maxAngularVelocity=0.3
     maxLinearAcceleration=700
     maxLinearVelocity=500
     velocityScaleZ=0
     worldAlpha=0.1
     externalLinearDragV=<0.5 0.5>
    }
   }
   n14 = "Transform (Modify) Bone - Bone: dyn_Hair_04"(ComponentPose=@n13, Alpha=@n2, Translation="0,0,0", Rotation="-1,5,0", Scale="1,1,1") {
    rotationMode=BMM_Additive
    rotationSpace=BCS_BoneSpace
    scaleMode=BMM_Additive
    scaleSpace=BCS_BoneSpace
    translationMode=BMM_Additive
    translationSpace=BCS_BoneSpace
    alphaBoolBlend {blendOption=Linear}
    boneToModify {boneName=dyn_Hair_04}
    rotation {pitch=-1, yaw=5}
   }
   n15 = Anim Dynamics - Bone: dyn_Hair_01(ComponentPose=Local To Component(LocalPose=Use cached pose 'BackpackCache'()), Alpha=0.2, ExternalForce="0,0,0") {
    angularDampingOverride=0.999
    bAngularSpring=true
    bGravityOverrideInSimSpace=true
    bOverrideAngularDamping=true
    bUseGravityOverride=true
    simulationSpace=RootRelative
    alphaBoolBlend {blendOption=Linear}
    boundBone {boneName=dyn_Hair_01}
    componentAppliedLinearAccClamp=<2500 2500 2500>
    componentLinearAccScale=<0.3 0.3 0.35>
    componentLinearVelScale=<0.35 0.35 0.35>
    gravityOverride {z=-600}
    physicsBodyDefinitions [{
      sphereCollisionRadius=0
      boundBone {boneName=dyn_Hair_01}
      boxExtents=<5 10 10>
      constraintSetup {angularTargetAxis=AxisY, angularLimitsMax=<5 1 20>, angularLimitsMin=<-5 -1 -7>, angularTarget=<0 0 0>}
      localJointOffset=<-5 -0.1 -0.5>
     }]
    relativeSpaceBone {boneName=dyn_Hair_base}
   }
   n16 = Anim Dynamics - Bone: dyn_Hair_05(ComponentPose=@n15, Alpha=0.2, ExternalForce="0,0,0") {
    angularDampingOverride=0.999
    bAngularSpring=true
    bGravityOverrideInSimSpace=true
    bOverrideAngularDamping=true
    bUseGravityOverride=true
    simulationSpace=RootRelative
    alphaBoolBlend {blendOption=Linear}
    boundBone {boneName=dyn_Hair_05}
    componentAppliedLinearAccClamp=<800 800 800>
    componentLinearAccScale=<0.3 0.3 0.35>
    componentLinearVelScale=<0.35 0.35 0.35>
    gravityOverride {z=-800}
    physicsBodyDefinitions [{
      sphereCollisionRadius=0
      boundBone {boneName=dyn_Hair_05}
      boxExtents=<10 5 10>
      constraintSetup {angularTargetAxis=AxisY, angularLimitsMax=<30 20 60>, angularLimitsMin=<-30 -20 -7>, angularTarget=<0 0 0>}
      localJointOffset=<-0.1 -5 1>
     }]
    relativeSpaceBone {boneName=dyn_Hair_03}
   }
   n17 = Anim Dynamics - Bone: dyn_Hair_02(ComponentPose=@n16, Alpha=0.2, ExternalForce="0,0,0") {
    angularDampingOverride=0.999
    bAngularSpring=true
    bGravityOverrideInSimSpace=true
    bOverrideAngularDamping=true
    bUseGravityOverride=true
    simulationSpace=RootRelative
    alphaBoolBlend {blendOption=Linear}
    boundBone {boneName=dyn_Hair_02}
    componentAppliedLinearAccClamp=<2500 2500 2500>
    componentLinearAccScale=<0.3 0.3 0.35>
    componentLinearVelScale=<0.35 0.35 0.35>
    gravityOverride {z=-600}
    physicsBodyDefinitions [{
      sphereCollisionRadius=0
      boundBone {boneName=dyn_Hair_02}
      boxExtents=<10 5 10>
      constraintSetup {angularTargetAxis=AxisY, angularLimitsMax=<5 1 45>, angularLimitsMin=<-5 -1 -7>, angularTarget=<0 0 0>}
      localJointOffset=<-2 -4 -0.5>
     }]
    relativeSpaceBone {boneName=dyn_Hair_01}
   }
   n18 = Anim Dynamics - Bone: dyn_Hair_06(ComponentPose=@n17, Alpha=0.2, ExternalForce="0,0,0") {
    angularDampingOverride=0.999
    bAngularSpring=true
    bGravityOverrideInSimSpace=true
    bOverrideAngularDamping=true
    bUseGravityOverride=true
    simulationSpace=RootRelative
    alphaBoolBlend {blendOption=Linear}
    boundBone {boneName=dyn_Hair_06}
    componentAppliedLinearAccClamp=<800 800 800>
    componentLinearAccScale=<0.3 0.3 0.35>
    componentLinearVelScale=<0.35 0.35 0.35>
    gravityOverride {z=-800}
    physicsBodyDefinitions [{
      sphereCollisionRadius=0
      boundBone {boneName=dyn_Hair_06}
      boxExtents=<10 5 10>
      constraintSetup {angularTargetAxis=AxisY, angularLimitsMax=<30 20 60>, angularLimitsMin=<-30 -20 -7>, angularTarget=<0 0 0>}
      localJointOffset=<-0.1 -5 1>
     }]
    relativeSpaceBone {boneName=dyn_Hair_03}
   }
   n19 = Anim Dynamics - Bone: dyn_Hair_03(ComponentPose=@n18, Alpha=0.2, ExternalForce="0,0,0") {
    angularDampingOverride=0.999
    bAngularSpring=true
    bGravityOverrideInSimSpace=true
    bOverrideAngularDamping=true
    bUseGravityOverride=true
    simulationSpace=RootRelative
    alphaBoolBlend {blendOption=Linear}
    boundBone {boneName=dyn_Hair_03}
    componentAppliedLinearAccClamp=<800 800 800>
    componentLinearAccScale=<0.3 0.3 0.35>
    componentLinearVelScale=<0.35 0.35 0.35>
    gravityOverride {z=-800}
    physicsBodyDefinitions [{
      sphereCollisionRadius=0
      boundBone {boneName=dyn_Hair_03}
      boxExtents=<10 5 10>
      constraintSetup {angularTargetAxis=AxisY, angularLimitsMax=<30 20 60>, angularLimitsMin=<-30 -20 -7>, angularTarget=<0 0 0>}
      localJointOffset=<-0.1 -5 1>
     }]
    relativeSpaceBone {boneName=dyn_Hair_02}
   }
   n20 = Anim Dynamics - Bone: dyn_Hair_07(ComponentPose=@n19, Alpha=0.2, ExternalForce="0,0,0") {
    angularDampingOverride=0.999
    bAngularSpring=true
    bGravityOverrideInSimSpace=true
    bOverrideAngularDamping=true
    bUseGravityOverride=true
    simulationSpace=RootRelative
    alphaBoolBlend {blendOption=Linear}
    boundBone {boneName=dyn_Hair_07}
    componentAppliedLinearAccClamp=<800 800 800>
    componentLinearAccScale=<0.3 0.3 0.35>
    componentLinearVelScale=<0.35 0.35 0.35>
    gravityOverride {z=-800}
    physicsBodyDefinitions [{
      sphereCollisionRadius=0
      boundBone {boneName=dyn_Hair_07}
      boxExtents=<10 5 10>
      constraintSetup {angularTargetAxis=AxisY, angularLimitsMax=<30 20 60>, angularLimitsMin=<-30 -20 -7>, angularTarget=<0 0 0>}
      localJointOffset=<-0.1 -5 1>
     }]
    relativeSpaceBone {boneName=dyn_Hair_03}
   }
   n21 = Anim Dynamics - Bone: dyn_Hair_04(ComponentPose=@n20, Alpha=0.2, ExternalForce="0,0,0") {
    angularDampingOverride=0.999
    bAngularSpring=true
    bGravityOverrideInSimSpace=true
    bOverrideAngularDamping=true
    bUseGravityOverride=true
    simulationSpace=RootRelative
    alphaBoolBlend {blendOption=Linear}
    boundBone {boneName=dyn_Hair_04}
    componentAppliedLinearAccClamp=<800 800 800>
    componentLinearAccScale=<0.3 0.3 0.35>
    componentLinearVelScale=<0.35 0.35 0.35>
    gravityOverride {z=-800}
    physicsBodyDefinitions [{
      sphereCollisionRadius=0
      boundBone {boneName=dyn_Hair_04}
      boxExtents=<10 5 10>
      constraintSetup {angularTargetAxis=AxisY, angularLimitsMax=<30 20 60>, angularLimitsMin=<-30 -20 -7>, angularTarget=<0 0 0>}
      localJointOffset=<-0.1 -5 1>
     }]
    relativeSpaceBone {boneName=dyn_Hair_03}
   }
   n22 = Anim Dynamics - Bone: dyn_Hair_08(ComponentPose=@n21, Alpha=0.2, ExternalForce="0,0,0") {
    angularDampingOverride=0.999
    bAngularSpring=true
    bGravityOverrideInSimSpace=true
    bOverrideAngularDamping=true
    bUseGravityOverride=true
    simulationSpace=RootRelative
    alphaBoolBlend {blendOption=Linear}
    boundBone {boneName=dyn_Hair_08}
    componentAppliedLinearAccClamp=<800 800 800>
    componentLinearAccScale=<0.3 0.3 0.35>
    componentLinearVelScale=<0.35 0.35 0.35>
    gravityOverride {z=-800}
    physicsBodyDefinitions [{
      sphereCollisionRadius=0
      boundBone {boneName=dyn_Hair_08}
      boxExtents=<10 5 10>
      constraintSetup {angularTargetAxis=AxisY, angularLimitsMax=<30 20 60>, angularLimitsMin=<-30 -20 -7>, angularTarget=<0 0 0>}
      localJointOffset=<-0.1 -5 1>
     }]
    relativeSpaceBone {boneName=dyn_Hair_03}
   }
   n23 = "Constraint - dyn_Hair_02 (RR)"(ComponentPose=@n22, Alpha=1, ConstraintWeights_0=0, ConstraintWeights_1=1) {
    alphaBoolBlend {blendOption=Linear}
    boneToModify {boneName=dyn_Hair_02}
    constraintSetup [{transformType=Rotation, targetBone {boneName=spine_05}}, {transformType=Rotation, targetBone {boneName=head}}]
    constraintWeights [1,1]
   }
   n24 = "Transform (Modify) Bone - Bone: dyn_Hair_02"(ComponentPose=@n23, Alpha=1, Translation="0,0,0", Rotation="0,-15,0", Scale="1,1,1") {
    rotationMode=BMM_Additive
    rotationSpace=BCS_BoneSpace
    scaleMode=BMM_Additive
    scaleSpace=BCS_BoneSpace
    translationMode=BMM_Additive
    translationSpace=BCS_BoneSpace
    alphaBoolBlend {blendOption=Linear}
    boneToModify {boneName=dyn_Hair_02}
    rotation {yaw=-15}
   }
   n25 = "Constraint - dyn_Hair_04 (RR)"(ComponentPose=@n24, Alpha=1, ConstraintWeights_0=0.7, ConstraintWeights_1=0.3) {
    alphaBoolBlend {blendOption=Linear}
    boneToModify {boneName=dyn_Hair_04}
    constraintSetup [{transformType=Rotation, targetBone {boneName=spine_05}}, {transformType=Rotation, targetBone {boneName=dyn_Hair_02}}]
    constraintWeights [1,1]
   }
   n26 = Get gravityOverride()
   n27 = RigidBody(ComponentPose=@n25, Alpha=Get Is NOT In Frontend(), OverrideWorldGravity=@n26) {
    bClampLinearTranslationLimitToRefPose=true
    bOverrideWorldGravity=true
    overridePhysicsAsset=&/F_MED_Keen_FaceAcc_Physics1
    simulationSpace=BaseBoneSpace
    alphaBoolBlend {blendOption=Linear}
    baseBoneRef {boneName=neck_01}
    componentAppliedLinearAccClamp=<800 800 800>
    componentLinearAccScale=<0.4 0.4 0.45>
    componentLinearVelScale=<0.45 0.45 0.45>
    overrideWorldGravity {z=-980}
    simSpaceSettings {
     maxAngularAcceleration=1000
     maxAngularVelocity=0.3
     maxLinearAcceleration=700
     maxLinearVelocity=500
     velocityScaleZ=0
     worldAlpha=0.1
     externalLinearDragV=<0.5 0.5>
    }
   }
   n28 = RigidBody(ComponentPose=@n27, Alpha=@n1, OverrideWorldGravity=@n26) {
    bClampLinearTranslationLimitToRefPose=true
    bOverrideWorldGravity=true
    overridePhysicsAsset=&F_MED_Keen_FaceAcc_Physics_Frontend
    simulationSpace=BaseBoneSpace
    alphaBoolBlend {blendOption=Linear}
    baseBoneRef {boneName=pelvis}
    componentAppliedLinearAccClamp=<5000 5000 5000>
    componentLinearAccScale=<2 2 2>
    componentLinearVelScale=<2 2 2>
    overrideWorldGravity {z=-980}
    simSpaceSettings {
     maxAngularAcceleration=1000
     maxAngularVelocity=0.3
     maxLinearAcceleration=700
     maxLinearVelocity=500
     velocityScaleZ=0
     worldAlpha=0.1
     externalLinearDragV=<0.5 0.5>
    }
   }
   n29 = "Transform (Modify) Bone - Bone: dyn_Hair_04"(ComponentPose=@n28, Alpha=@n1, Translation="0,0,0", Rotation="-1,5,0", Scale="1,1,1") {
    rotationMode=BMM_Additive
    rotationSpace=BCS_BoneSpace
    scaleMode=BMM_Additive
    scaleSpace=BCS_BoneSpace
    translationMode=BMM_Additive
    translationSpace=BCS_BoneSpace
    alphaBoolBlend {blendOption=Linear}
    boneToModify {boneName=dyn_Hair_04}
    rotation {pitch=-1, yaw=5}
   }
   n30 = "Transform (Modify) Bone - Bone: dyn_Hair_03"(ComponentPose=@n29, Alpha=Get PC-PerfMode() -> (PC-PerfMode:double), Translation="0,0,0", Rotation="-1,20,0", Scale="1,1,1") {
    rotationMode=BMM_Additive
    rotationSpace=BCS_BoneSpace
    scaleMode=BMM_Additive
    scaleSpace=BCS_BoneSpace
    translationMode=BMM_Additive
    translationSpace=BCS_BoneSpace
    alphaBoolBlend {blendOption=Linear}
    boneToModify {boneName=dyn_Hair_03}
    rotation {pitch=-1, yaw=20}
   }
   n31 = "Transform (Modify) Bone - Bone: dyn_Hair_03"(ComponentPose=@n30, Alpha=Get Is NOT In Frontend(), Translation="0,0,0", Rotation="1,-20,0", Scale="1,1,1") {
    rotationMode=BMM_Additive
    rotationSpace=BCS_BoneSpace
    scaleMode=BMM_Additive
    scaleSpace=BCS_BoneSpace
    translationMode=BMM_Additive
    translationSpace=BCS_BoneSpace
    alphaBoolBlend {blendOption=Linear}
    boneToModify {boneName=dyn_Hair_03}
    rotation {pitch=1, yaw=-20}
   }
   Output Pose(Result=Blend Poses by bool(
     BlendPose_0=Component To Local(ComponentPose=@n31)
     BlendPose_1=Component To Local(ComponentPose=@n14)
     bActiveValue=Get isWearingPackBool() -> (isWearingPackBool:bool)
    ) {blendType=Linear}) {layerGroup=None}
   n32 = Get Skel_Mesh()
   n33 = "Constraint - dyn_Hair_03 (RR)"!(
    ComponentPose=Local To Component(LocalPose=Copy Pose From Mesh(SourceMeshComponent=@n32) {bUseAttachedParent=true})
    Alpha=1
    ConstraintWeights_0=0.5
    ConstraintWeights_1:float=0.5
   ) -> (Pose) {
    alphaBoolBlend {blendOption=Linear}
    boneToModify {boneName=dyn_Hair_03}
    constraintSetup [{transformType=Rotation, targetBone {boneName=spine_05}}, {transformType=Rotation, targetBone {boneName=dyn_Hair_02}}]
    constraintWeights [1,1]
   }
   n34 = "Constraint - dyn_Hair_l_03 (RR)"!(ComponentPose=@n33, Alpha=1, ConstraintWeights_0=0.8, ConstraintWeights_1:float=0.2) -> (Pose) {
    alphaBoolBlend {blendOption=Linear}
    boneToModify {boneName=dyn_Hair_l_03}
    constraintSetup [
     {transformType=Rotation, perAxis {bX=false, bZ=false}, targetBone {boneName=spine_05}}
     {offsetOption=None, transformType=Rotation, perAxis {bX=false, bZ=false}, targetBone {boneName=dyn_Hair_l_02}}
    ]
    constraintWeights [1,1]
   }
   n35 = "Constraint - dyn_Hair_r_03 (RR)"!(ComponentPose=@n34, Alpha=1, ConstraintWeights_0=0.8, ConstraintWeights_1:float=0.2) -> (Pose) {
    alphaBoolBlend {blendOption=Linear}
    boneToModify {boneName=dyn_Hair_r_03}
    constraintSetup [
     {transformType=Rotation, perAxis {bZ=false}, targetBone {boneName=spine_05}}
     {offsetOption=None, transformType=Rotation, perAxis {bZ=false}, targetBone {boneName=dyn_Hair_r_02}}
    ]
    constraintWeights [1,1]
   }
   n36 = Anim Dynamics - Bone: dyn_Hair_l_01(ComponentPose=@n35, Alpha=0.5, ExternalForce="0,0,0") {
    angularDampingOverride=0.999
    bAngularSpring=true
    bGravityOverrideInSimSpace=true
    bOverrideAngularDamping=true
    bUseGravityOverride=true
    simulationSpace=RootRelative
    alphaBoolBlend {blendOption=Linear}
    boundBone {boneName=dyn_Hair_l_01}
    componentAppliedLinearAccClamp=<2500 2500 2500>
    componentLinearAccScale=<0.3 0.3 0.35>
    componentLinearVelScale=<0.35 0.35 0.35>
    gravityOverride {z=-600}
    physicsBodyDefinitions [{
      sphereCollisionRadius=0
      boundBone {boneName=dyn_Hair_l_01}
      boxExtents=<5 10 10>
      constraintSetup {angularTargetAxis=AxisY, angularLimitsMax=<5 5 10>, angularLimitsMin=<-2 -2 -25>, angularTarget=<0 0 0>}
      localJointOffset=<-5 -0.1 -1>
     }]
    relativeSpaceBone {boneName=spine_05}
   }
   n37 = Anim Dynamics - Bone: dyn_Hair_r_01(ComponentPose=@n36, Alpha=0.5, ExternalForce="0,0,0") {
    angularDampingOverride=0.999
    bAngularSpring=true
    bGravityOverrideInSimSpace=true
    bOverrideAngularDamping=true
    bUseGravityOverride=true
    simulationSpace=RootRelative
    alphaBoolBlend {blendOption=Linear}
    boundBone {boneName=dyn_Hair_r_01}
    componentAppliedLinearAccClamp=<2500 2500 2500>
    componentLinearAccScale=<0.3 0.3 0.35>
    componentLinearVelScale=<0.35 0.35 0.35>
    gravityOverride {z=-600}
    physicsBodyDefinitions [{
      sphereCollisionRadius=0
      boundBone {boneName=dyn_Hair_r_01}
      boxExtents=<5 10 10>
      constraintSetup {angularTargetAxis=AxisY, angularLimitsMax=<5 5 10>, angularLimitsMin=<-2 -2 -25>, angularTarget=<0 0 0>}
      localJointOffset=<5 -0.1 -1>
     }]
    relativeSpaceBone {boneName=spine_05}
   }
   n38 = Anim Dynamics - Bone: dyn_Hair_l_02(ComponentPose=@n37, Alpha=0.5, ExternalForce="0,0,0") {
    angularDampingOverride=0.999
    bAngularSpring=true
    bGravityOverrideInSimSpace=true
    bOverrideAngularDamping=true
    bUseGravityOverride=true
    simulationSpace=RootRelative
    alphaBoolBlend {blendOption=Linear}
    boundBone {boneName=dyn_Hair_l_02}
    componentAppliedLinearAccClamp=<2500 2500 2500>
    componentLinearAccScale=<0.3 0.3 0.35>
    componentLinearVelScale=<0.35 0.35 0.35>
    gravityOverride {z=-600}
    physicsBodyDefinitions [{
      sphereCollisionRadius=0
      boundBone {boneName=dyn_Hair_l_02}
      boxExtents=<5 10 10>
      constraintSetup {angularTargetAxis=AxisY, angularLimitsMax=<5 5 10>, angularLimitsMin=<-2 -2 -25>, angularTarget=<0 0 0>}
      localJointOffset=<-5 -0.1 1>
     }]
    relativeSpaceBone {boneName=spine_05}
   }
   n39 = Anim Dynamics - Bone: dyn_Hair_r_02(ComponentPose=@n38, Alpha=0.5, ExternalForce="0,0,0") {
    angularDampingOverride=0.999
    bAngularSpring=true
    bGravityOverrideInSimSpace=true
    bOverrideAngularDamping=true
    bUseGravityOverride=true
    simulationSpace=RootRelative
    alphaBoolBlend {blendOption=Linear}
    boundBone {boneName=dyn_Hair_r_02}
    componentAppliedLinearAccClamp=<2500 2500 2500>
    componentLinearAccScale=<0.3 0.3 0.35>
    componentLinearVelScale=<0.35 0.35 0.35>
    gravityOverride {z=-600}
    physicsBodyDefinitions [{
      sphereCollisionRadius=0
      boundBone {boneName=dyn_Hair_r_02}
      boxExtents=<5 10 10>
      constraintSetup {angularTargetAxis=AxisY, angularLimitsMax=<5 5 10>, angularLimitsMin=<-2 -2 -25>, angularTarget=<0 0 0>}
      localJointOffset=<5 -0.1 1>
     }]
    relativeSpaceBone {boneName=spine_05}
   }
   n40 = Anim Dynamics - Bone: dyn_Hair_l_03(ComponentPose=@n39, Alpha=0.5, ExternalForce="0,0,0") {
    angularDampingOverride=0.999
    bAngularSpring=true
    bGravityOverrideInSimSpace=true
    bOverrideAngularDamping=true
    bUseGravityOverride=true
    simulationSpace=RootRelative
    alphaBoolBlend {blendOption=Linear}
    boundBone {boneName=dyn_Hair_l_03}
    componentAppliedLinearAccClamp=<2500 2500 2500>
    componentLinearAccScale=<0.3 0.3 0.35>
    componentLinearVelScale=<0.35 0.35 0.35>
    gravityOverride {z=-600}
    physicsBodyDefinitions [{
      sphereCollisionRadius=0
      boundBone {boneName=dyn_Hair_l_03}
      boxExtents=<5 10 10>
      constraintSetup {angularTargetAxis=AxisY, angularLimitsMax=<5 5 10>, angularLimitsMin=<-2 -2 -25>, angularTarget=<0 0 0>}
      localJointOffset=<-5 -0.1 1>
     }]
    relativeSpaceBone {boneName=dyn_Hair_l_02}
   }
   n41 = Anim Dynamics - Bone: dyn_Hair_r_03(ComponentPose=@n40, Alpha=0.5, ExternalForce="0,0,0") {
    angularDampingOverride=0.999
    bAngularSpring=true
    bGravityOverrideInSimSpace=true
    bOverrideAngularDamping=true
    bUseGravityOverride=true
    simulationSpace=RootRelative
    alphaBoolBlend {blendOption=Linear}
    boundBone {boneName=dyn_Hair_r_03}
    componentAppliedLinearAccClamp=<2500 2500 2500>
    componentLinearAccScale=<0.3 0.3 0.35>
    componentLinearVelScale=<0.35 0.35 0.35>
    gravityOverride {z=-600}
    physicsBodyDefinitions [{
      sphereCollisionRadius=0
      boundBone {boneName=dyn_Hair_r_03}
      boxExtents=<5 10 10>
      constraintSetup {angularTargetAxis=AxisY, angularLimitsMax=<5 5 10>, angularLimitsMin=<-2 -2 -25>, angularTarget=<0 0 0>}
      localJointOffset=<5 -0.1 1>
     }]
    relativeSpaceBone {boneName=dyn_Hair_r_02}
   }
   n42 = Anim Dynamics - Bone: dyn_Hair_l_04(ComponentPose=@n41, Alpha=0.5, ExternalForce="0,0,0") {
    angularDampingOverride=0.999
    bAngularSpring=true
    bGravityOverrideInSimSpace=true
    bOverrideAngularDamping=true
    bUseGravityOverride=true
    simulationSpace=RootRelative
    alphaBoolBlend {blendOption=Linear}
    boundBone {boneName=dyn_Hair_l_04}
    componentAppliedLinearAccClamp=<2500 2500 2500>
    componentLinearAccScale=<0.3 0.3 0.35>
    componentLinearVelScale=<0.35 0.35 0.35>
    gravityOverride {z=-600}
    physicsBodyDefinitions [{
      sphereCollisionRadius=0
      boundBone {boneName=dyn_Hair_l_04}
      boxExtents=<5 10 10>
      constraintSetup {angularTargetAxis=AxisY, angularLimitsMax=<5 5 10>, angularLimitsMin=<-2 -2 -25>, angularTarget=<0 0 0>}
      localJointOffset=<-5 -1 1>
     }]
    relativeSpaceBone {boneName=spine_05}
   }
   n43 = Anim Dynamics - Bone: dyn_Hair_r_04(ComponentPose=@n42, Alpha=0.5, ExternalForce="0,0,0") {
    angularDampingOverride=0.999
    bAngularSpring=true
    bGravityOverrideInSimSpace=true
    bOverrideAngularDamping=true
    bUseGravityOverride=true
    simulationSpace=RootRelative
    alphaBoolBlend {blendOption=Linear}
    boundBone {boneName=dyn_Hair_r_04}
    componentAppliedLinearAccClamp=<2500 2500 2500>
    componentLinearAccScale=<0.3 0.3 0.35>
    componentLinearVelScale=<0.35 0.35 0.35>
    gravityOverride {z=-600}
    physicsBodyDefinitions [{
      sphereCollisionRadius=0
      boundBone {boneName=dyn_Hair_r_04}
      boxExtents=<5 10 10>
      constraintSetup {angularTargetAxis=AxisY, angularLimitsMax=<5 5 10>, angularLimitsMin=<-2 -2 -25>, angularTarget=<0 0 0>}
      localJointOffset=<5 -1 1>
     }]
    relativeSpaceBone {boneName=spine_05}
   }
   n44 = Anim Dynamics - Bone: dyn_Hair_l_05(ComponentPose=@n43, Alpha=0.5, ExternalForce="0,0,0") {
    angularDampingOverride=0.999
    bAngularSpring=true
    bGravityOverrideInSimSpace=true
    bOverrideAngularDamping=true
    bUseGravityOverride=true
    simulationSpace=RootRelative
    alphaBoolBlend {blendOption=Linear}
    boundBone {boneName=dyn_Hair_l_05}
    componentAppliedLinearAccClamp=<2500 2500 2500>
    componentLinearAccScale=<0.3 0.3 0.35>
    componentLinearVelScale=<0.35 0.35 0.35>
    gravityOverride {z=-600}
    physicsBodyDefinitions [{
      sphereCollisionRadius=0
      boundBone {boneName=dyn_Hair_l_05}
      boxExtents=<5 10 10>
      constraintSetup {angularTargetAxis=AxisY, angularLimitsMax=<5 5 10>, angularLimitsMin=<-2 -2 -25>, angularTarget=<0 0 0>}
      localJointOffset=<-5 -1 1>
     }]
    relativeSpaceBone {boneName=spine_05}
   }
   n45 = Anim Dynamics - Bone: dyn_Hair_r_05(ComponentPose=@n44, Alpha=0.5, ExternalForce="0,0,0") {
    angularDampingOverride=0.999
    bAngularSpring=true
    bGravityOverrideInSimSpace=true
    bOverrideAngularDamping=true
    bUseGravityOverride=true
    simulationSpace=RootRelative
    alphaBoolBlend {blendOption=Linear}
    boundBone {boneName=dyn_Hair_r_05}
    componentAppliedLinearAccClamp=<2500 2500 2500>
    componentLinearAccScale=<0.3 0.3 0.35>
    componentLinearVelScale=<0.35 0.35 0.35>
    gravityOverride {z=-600}
    physicsBodyDefinitions [{
      sphereCollisionRadius=0
      boundBone {boneName=dyn_Hair_r_05}
      boxExtents=<5 10 10>
      constraintSetup {angularTargetAxis=AxisY, angularLimitsMax=<5 5 10>, angularLimitsMin=<-2 -2 -25>, angularTarget=<0 0 0>}
      localJointOffset=<5 -1 1>
     }]
    relativeSpaceBone {boneName=spine_05}
   }
   n46 = Get Shoulder_rot_f_r() -> (Shoulder_rot_f_r:double)
   n47 = "Transform (Modify) Bone - Bone: dyn_Hair_r_base"(ComponentPose=@n45, Alpha=@n46, Translation="0,0,0", Rotation="-15,0,0", Scale="1,1,1") {
    rotationMode=BMM_Additive
    rotationSpace=BCS_BoneSpace
    scaleMode=BMM_Additive
    scaleSpace=BCS_BoneSpace
    translationMode=BMM_Additive
    translationSpace=BCS_BoneSpace
    alphaBoolBlend {blendOption=Linear}
    boneToModify {boneName=dyn_Hair_r_base}
    rotation {pitch=-15}
   }
   n48 = "Transform (Modify) Bone - Bone: dyn_Hair_r_02"(ComponentPose=@n47, Alpha=@n46, Translation="0,0,0", Rotation="10,0,0", Scale="1,1,1") {
    rotationMode=BMM_Additive
    rotationSpace=BCS_BoneSpace
    scaleMode=BMM_Additive
    scaleSpace=BCS_BoneSpace
    translationMode=BMM_Additive
    translationSpace=BCS_BoneSpace
    alphaBoolBlend {blendOption=Linear}
    boneToModify {boneName=dyn_Hair_r_02}
    rotation {pitch=10}
   }
   Save cached pose 'BackpackCache'(Pose=Component To Local(ComponentPose=@n48)) {cacheName=BackpackCache}
   Get Is In Frontend()
  ]}
 EventGraph {nodes [
   Event BlueprintUpdateAnimation()
   n1 = Get Skel_Mesh()
   > Set neck_rot_up(neck_rot_up:double=GetAlphaFromReferenceTransformDelta(SkeletalMeshComponent=@n1, BoneName=head, BaseName=spine_05, InputRange=(X=-50,Y=-90), TwistAxis="0,0,1")) -> (Output_Get:double)
   > Set neck_rot_dn(neck_rot_dn:double=GetAlphaFromReferenceTransformDelta(SkeletalMeshComponent=@n1, BoneName=neck_01, BaseName=spine_05, DeltaComponentType=EulerX, InputRange=(X=-15,Y=50), TwistAxis="0,0,1")) -> (Output_Get:double)
   n2 = Get Shoulder_Mesh() -> (Shoulder_Mesh:Skeletal Mesh Component Object Reference)
   > Set Shoulder_rot_f_r(Shoulder_rot_f_r:double=GetAlphaFromReferenceTransformDelta(SkeletalMeshComponent=@n2, BoneName=upperarm_r, BaseName=clavicle_r, InputRange=(X=0,Y=90), TwistAxis="0,0,1")) -> (Output_Get:double)
   > Set Shoulder_rot_f_l(Shoulder_rot_f_l:double=GetAlphaFromReferenceTransformDelta(SkeletalMeshComponent=@n2, BoneName=upperarm_l, BaseName=clavicle_l, InputRange=(X=0,Y=90), TwistAxis="0,0,1")) -> (Output_Get:double)
   n3 = Get bIsSkydiveFloating()
   n4 = Get bIsSkydiveDiveMode()
   > Set gravityOverride(gravityOverride:Vector=SelectVector(A="-10, 0, 0", B="0, 0,-980", bPickA=OR Boolean(A=@n3, B=@n4))) -> (Output_Get:Vector)
   n5 = Get bIsInFrontEnd()
   > Set Is In Frontend(Is In Frontend:double="To Float (Boolean)"(InBool=@n5)) -> (Output_Get:double)
   > Set Is NOT In Frontend(Is NOT In Frontend:double="To Float (Boolean)"(InBool=NOT Boolean(A=@n5))) -> (Output_Get:double)
   n6 = Get bRigidBodyEnabled()
   > Set PC-PerfMode(PC-PerfMode:double="To Float (Boolean)"(InBool=NOT Boolean(A=@n6))) -> (Output_Get:double)
   Event OnCharacterCustomizationLoaded()
   n7 = CheckForBackbling()
   > Set isWearingPackBool(isWearingPackBool:bool=AND Boolean(A=@n7, B=NOT Boolean(A=@n7.bIsBackblingLong))) -> (Output_Get:bool)
   Event BlueprintInitializeAnimation()
   > Set Skel_Mesh(Skel_Mesh=GetOwnerSkeletalMeshComponent())
   > n8 = Cast To PlayerPawn_Athena(Object=TryGetPawnOwner())
   > n9 = Set Pawn Owner(Pawn Owner:Fort Player Pawn Athena Object Reference=@n8) -> (Output_Get:Fort Player Pawn Athena Object Reference)
   > Set Shoulder_Mesh(Shoulder_Mesh:Skeletal Mesh Component Object Reference=GetSkeletalMeshForPartType(self=@n9)) -> (Output_Get:Skeletal Mesh Component Object Reference)
  ]}
]
variables [
 Dampening {type=number}
 dynAlpha_noBackbling {type=number}
 dynAlpha_withBackbling {type=number}
 gravityOverride {type=Vector}
 Is In Frontend {type=number}
 Is NOT In Frontend {type=number}
 isWearingPackBool {type=boolean}
 linAcc {type=Vector}
 linVelocity {type=Vector}
 neck_rot_dn {type=number}
 neck_rot_up {type=number}
 Pawn Owner {type=FortPlayerPawnAthena}
 PC-PerfMode {type=number}
 Shoulder_Mesh {type=SkeletalMeshComponent}
 Shoulder_rot_f_l {type=number}
 Shoulder_rot_f_r {type=number}
 Skel_Mesh {type=SkeletalMeshComponent}
]
